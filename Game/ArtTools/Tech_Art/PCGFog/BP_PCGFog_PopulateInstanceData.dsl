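; PCG Blueprint element that converts point data into packed instance data for the owning BP_PCGFog_Volume, then returns its input unchanged.
; Each point becomes a Vector4: XYZ holds rounded location as the integer part and remapped up vector, capped below one, as a sign-aware fraction.
; W holds the largest scale component.
; The array is resized and written to the volume per spatial input, so each input overwrites the previous one.
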(fn ExecuteWithContext
  (params (InContext :type PCGContext%20Structure%20%28by%20ref%29) (Input :type PCGData%20Collection%20Structure%20%28by%20ref%29))
  (bind _returnvalue (PCG|Temporary|GetComponent InContext))
  (bind _asbp_pcgfog_volume (Utilities|Casting|CastToBP_PCGFog_Volume (Components|GetOwner _returnvalue))
    (:then
      (LocalVariables|SetParent _asbp_pcgfog_volume)
      (bind _returnvalue_2 (Data|GetAllSpatialInputs Input))
      (for _array_element _returnvalue_2
        (bind _aspcgpoint_data (Utilities|Casting|CastToPCGPointData (PCG|TaggedData|BreakTaggedData _array_element)))
        (Utilities|Array|Resize (LocalVariables|GetInstanceData) (PointData|GetNumPoints _aspcgpoint_data))
        (for (_array_element_2 _array_index) (SpatialData|GetPoints _aspcgpoint_data)
          (bind _array_element_transform (Utilities|Struct|BreakPCGPoint _array_element_2))
          (bind (_array_element_transform_location _array_element_transform_rotation _array_element_transform_scale) (Math|Transform|BreakTransform _array_element_transform))
          (Utilities|Array|SetArrayElem (LocalVariables|GetInstanceData) _array_index (Math|Vector4|MakeVector4 (Math|Float|SelectFloat (- (DisplayHelper|RoundFloat (.x _array_element_transform_location) "HalfFromZero") (.x (Math|Vector|VectorComponentMin (* (+ (Math|Vector|GetUpVector _array_element_transform_rotation) 1.0) 0.5) (Math|Vector|MakeVector 0.999 0.999 0.999)))) (+ (DisplayHelper|RoundFloat (.x _array_element_transform_location) "HalfFromZero") (.x (Math|Vector|VectorComponentMin (* (+ (Math|Vector|GetUpVector _array_element_transform_rotation) 1.0) 0.5) (Math|Vector|MakeVector 0.999 0.999 0.999)))) (< (DisplayHelper|RoundFloat (.x _array_element_transform_location) "HalfFromZero") 0)) (Math|Float|SelectFloat (- (DisplayHelper|RoundFloat (.y _array_element_transform_location) "HalfFromZero") (.y (Math|Vector|VectorComponentMin (* (+ (Math|Vector|GetUpVector _array_element_transform_rotation) 1.0) 0.5) (Math|Vector|MakeVector 0.999 0.999 0.999)))) (+ (DisplayHelper|RoundFloat (.y _array_element_transform_location) "HalfFromZero") (.y (Math|Vector|VectorComponentMin (* (+ (Math|Vector|GetUpVector _array_element_transform_rotation) 1.0) 0.5) (Math|Vector|MakeVector 0.999 0.999 0.999)))) (< (DisplayHelper|RoundFloat (.y _array_element_transform_location) "HalfFromZero") 0)) (Math|Float|SelectFloat (- (DisplayHelper|RoundFloat (.z _array_element_transform_location) "HalfFromZero") (.z (Math|Vector|VectorComponentMin (* (+ (Math|Vector|GetUpVector _array_element_transform_rotation) 1.0) 0.5) (Math|Vector|MakeVector 0.999 0.999 0.999)))) (+ (DisplayHelper|RoundFloat (.z _array_element_transform_location) "HalfFromZero") (.z (Math|Vector|VectorComponentMin (* (+ (Math|Vector|GetUpVector _array_element_transform_rotation) 1.0) 0.5) (Math|Vector|MakeVector 0.999 0.999 0.999)))) (< (DisplayHelper|RoundFloat (.z _array_element_transform_location) "HalfFromZero") 0)) (Math|Vector|GetMaxElement _array_element_transform_scale))))
        (Variables|/Game/ArtTools/Tech_Art/PCGFog/BP_PCGFog_Volume.BP_PCGFog_Volume_C|SetInstanceData (LocalVariables|GetInstanceData) (LocalVariables|GetParent))))
    (:CastFailed
      _))
  (return Input))
