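; PCG fog volume Blueprint that feeds stored per-instance data to a Niagara system for rendering fog.
; The construction script applies the fog material to Niagara and updates it when instance data exists.
; UpdateInstanceData pushes the Vector4 array into Niagara and activates it with a reset; its PCG data parameter is unused.
; BeginPlay repeats that update, then clears the Blueprint's array.
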(event EventBeginPlay
  (CallFunction|UpdateInstanceData)
  (Utilities|Array|Clear (Variables|Internal|GetInstanceData)))

(fn UpdateInstanceData
  (params (PCGData :type PCGData%20Collection%20Structure))
  (bind _niagara (Variables|Default|GetNiagara))
  (Niagara|NiagaraSetVector4Array _niagara "InstanceData" (Variables|Internal|GetInstanceData))
  (Components|Activation|Activate _niagara true))

(fn ConstructionScript
  (Niagara|SetNiagaraVariable(Material) (Variables|Default|GetNiagara) "FogMaterial" (Variables|Default|GetFog%20material))
  (if (Utilities|Array|IsNotEmpty (Variables|Internal|GetInstanceData))
    (CallFunction|UpdateInstanceData)))
