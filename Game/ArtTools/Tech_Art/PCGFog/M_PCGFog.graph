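# M_PCGFog is an additive Volume-domain material for Niagara particle fog, tinted by a time-of-day curve atlas.
# Density combines a radial sphere mask, a height curve tilted by a slope vector in particle color, near-camera fade and panning texture.
# Switches add distortion, slope-following motion, wind gusts and continuous cross-particle coordinates.
# Epic quality allows costly micro noise jittered per pixel and temporal sample, and skips the color desaturation.
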Material /Game/ArtTools/Tech_Art/PCGFog/M_PCGFog
graph {
 properties {
  bUsedWithNiagaraMeshParticles=true
  bUsedWithNiagaraSprites=true
  blendMode=BLEND_Additive
  materialDomain=MD_Volume
  pixelDepthOffsetMode=PDOM_Legacy
  usedShadingModels=Default Lit
 }
 nodes [
  n1 = CurveAtlasRowParameter(CurveTime=CollectionParameter() {collection=&FortniteMaterialParameters, parameterName=TimeOfDay_0-1}) {
   atlas=&/CA_PCGFog
   curve=&/CC_PCGFog_Color_Hermes_MossyForest
   defaultValue=3
   group=Color
   parameterName=Color Curve
  }
  n2 = Desaturation(None=@n1, Fraction=ScalarParameter() {defaultValue=-0.3, group=Color, parameterName="Desaturation (Non-Lumen)"})
  n3 = ScalarParameter() {group=Color, parameterName=Emissive Fraction}
  n4 = ParticlePositionWS() {originType=CameraRelative}
  n5 = WorldPosition() {worldPositionShaderOffset=WPT_CameraRelative}
  n6 = ParticleRadius()
  n7 = SphereMask(A=@n4, B=@n5.XYZ, Radius=@n6, Hardness=OneMinus(None=ScalarParameter() {defaultValue=0.1, group=Falloff, parameterName=Radial Falloff})) {hardnessPercent=0}
  n8 = Constant3Vector() {constant=<rgba 0 0 1 1>}
  n9 = ParticleColor()
  n10 = Multiply(A=ComponentMask(None=@n9.RGB) {g=true, r=true}) {constB=-1}
  n11 = AppendVector(A=@n10, B=ComponentMask(None=@n9.RGB) {b=true})
  n12 = DotProduct(A=@n8, B=@n11)
  n13 = Multiply(A=Power(Base=ArccosineFast!(None=@n12)) {constExponent=0.25}) {constB=57.29, desc=Radians to Degrees}
  n14 = Add(
   A=RotateAboutAxis(
    NormalizedRotationAxis=CrossProduct(A=@n8, B=@n11)
    RotationAngle=Divide(A=@n13) {constB=360, desc=Degrees to 0-1}
    PivotPoint=@n4
    Position=@n5.XYZ
   )
   B=@n5.XYZ
  )
  n15 = Divide(A=Subtract(A=ComponentMask(None=@n14) {b=true}, B=ComponentMask(None=@n4) {b=true}), B=@n6)
  n16 = CurveAtlasRowParameter(CurveTime=@n15) {
   atlas=&CA_PCGFog
   curve=&CC_PCGFog_Color_Hermes_MossyForest
   defaultValue=3
   group=Falloff
   parameterName=Density Curve
  }
  n17 = NamedRerouteDeclaration(None=StaticComponentMaskParameter(None=AppendVector(A=@n16, B=@n16)) {defaultR=true, group=Falloff, parameterName=Density Curve Channel}) {name=Density Curve, nodeColor=<rgba 0.0312 1 0 1>}
  n18 = Length(None=WorldPosition() {worldPositionShaderOffset=WPT_CameraRelative}.XYZ) {is=class MaterialExpressionLength}
  n19 = ScalarParameter() {defaultValue=1000, group=Falloff, parameterName=Near Fade Distance}
  n20 = Power(Base=Saturate(None=Divide(A=@n18, B=@n19)), Exp=ScalarParameter() {defaultValue=1, group=Falloff, parameterName=Near Fade Exponent}) {constExponent=16}
  n21 = Multiply(
   A=NamedRerouteUsage()
   B=Multiply(A=Multiply(A=@n7, B=@n17), B=NamedRerouteDeclaration(None=@n20) {name=Near Gradient, nodeColor=<rgba 0.2891 0 1 1>})
  ) {constA=1}
  Multiply(A=Multiply(A=QualitySwitch(Default=@n2, Epic=@n1), B=Multiply(A=@n1.A, B=@n3)), B=@n21)
  Multiply(A=@n1.A, B=@n21)
  n22 = ScalarParameter() {defaultValue=0.1, group=Motion, parameterName=Gust Reset Density, sortPriority=33}
  n23 = VectorParameter() {group=Motion, parameterName=Density Texture Motion, sortPriority=20, defaultValue=<rgba 0.02 0.02 -0.02 1>}
  n24 = StaticSwitchParameter(
   True=Multiply(A=NamedRerouteUsage(), B=ScalarParameter() {defaultValue=0.2, group=Motion, parameterName=Density Texture Speed, sortPriority=20})
   False=ComponentMask(None=@n23.RGB) {g=true, r=true}
  ) {desc=Motion will follow the slope of the surface below, group=Motion, parameterName=Use Slope Based Motion, sortPriority=10}
  n25 = Multiply(A=@n24, B=Time())
  n26 = StaticSwitchParameter(True=Add(A=NamedRerouteUsage(), B=@n25), False=@n25) {group=Motion, parameterName=Use Wind Gusts, sortPriority=30}
  n27 = VectorParameter() {group=Motion, parameterName=Distortion Motion, sortPriority=30, defaultValue=<rgba 0.01 0.01 -0.01 1>}
  n28 = StaticSwitchParameter(
   True=Multiply(A=NamedRerouteUsage(), B=ScalarParameter() {defaultValue=0.2, group=Motion, parameterName=Distortion Speed, sortPriority=21})
   False=ComponentMask(None=@n27.RGB) {g=true, r=true}
  ) {desc=Motion will follow the slope of the surface below, group=Motion, parameterName=Use Slope Based Motion, sortPriority=10}
  n29 = Multiply(A=Time(), B=@n28)
  n30 = Divide(A=NamedRerouteUsage(), B=ScalarParameter() {defaultValue=1024, group=Texture, parameterName=Distortion Scale, sortPriority=-40})
  n31 = StaticComponentMaskParameter(None=TextureSampleParameter2D(UVs=Add(A=@n29, B=@n30)) {group=Texture, parameterName=Distortion Texture, sortPriority=-50, texture=/Engine/MaterialTemplates/Textures/T_Noise01}.RGBA) {defaultR=true, group=Texture, parameterName=Distortion Texture Channel, sortPriority=-45}
  n32 = Add(
   A=Multiply(
    A=ConstantBiasScale(None=@n31) {bias=-0.5, scale=2}
    B=ScalarParameter() {defaultValue=0.5, group=Texture, parameterName=Distortion Intensity, sortPriority=-35}
   )
   B=Divide(A=NamedRerouteUsage(), B=ScalarParameter() {defaultValue=100, group=Texture, parameterName=Density Scale, sortPriority=-90})
  )
  n33 = TextureSampleParameter2D(UVs=StaticSwitchParameter(True=Add(A=@n26, B=@n32), False=@n26) {group=Texture, parameterName=Use Distortion, sortPriority=-78}) {
   group=Texture
   parameterName=Density Texture
   samplerSource=SSM_Wrap_WorldGroupSettings
   sortPriority=-100
   texture=/Game/Effects/Fort_Effects/Textures/T_PinkSpike_Voronoi2D_Height
  }
  n34 = StaticComponentMaskParameter(None=@n33.RGBA) {defaultR=true, group=Texture, parameterName=Density Texture Channel, sortPriority=-95}
  n35 = NamedRerouteUsage()
  n36 = VectorParameter() {
   group=Texture
   parameterName=DensityTextureLevels
   sortPriority=-80
   channelNames {a=Middlepoint, b=WhiteValue, g=MiddleValue, r=BlackValue}
   defaultValue=<rgba 0 0.5 1 0.5>
  }
  n37 = MaterialFunctionCall!(
   Texture=StaticSwitchParameter(True=LinearInterpolate(A=@n22, B=@n34, Alpha=@n35), False=@n34) {group=Motion, parameterName=Use Wind Gusts, sortPriority=30}
   ---------------
   New Black Value=@n36.BlackValue
   New Middle Value=@n36.BlackValue
   New White Value=@n36.BlackValue
   Middle Point=@n36.BlackValue
   "--------------- "
   Define Interpolation Curve
   Interpolation Power
   Invert Interpolation Power
  ) {materialFunction=/Engine/Functions/Engine_MaterialFunctions02/3PointLevels}
  n38 = ScalarParameter() {defaultValue=0.01, group=Texture, parameterName=Micro Noise Scale, sortPriority=61}
  n39 = ViewProperty() {property=MEVP_TemporalSampleIndex}
  MaterialExpressionAdd_20 = Add(A=ScreenPosition().PixelPosition, B=AppendVector(A=@n39.Property, B=@n39.Property))
  n40 = Custom!(p=@MaterialExpressionAdd_20) {
   code="Mod( ((uint)(p.x) + 2 * (uint)(p.y)) , 5 )"
   outputType=CMOT_Float1
   inputs [{inputName=p, input {expression=&self:MaterialExpressionAdd_20}}]
  }
  n41 = Multiply(
   A=Multiply(A=@n38, B=ScalarParameter() {defaultValue=0.2, group=Texture, parameterName=Micro Noise Smoothness, sortPriority=68})
   B=ConstantBiasScale(None=@n40) {bias=-2}
  )
  n42 = NamedRerouteUsage()
  n43 = Sine(None=Multiply(A=Add(A=@n38, B=@n41), B=ComponentMask(None=@n42) {g=true, r=true}))
  n44 = Multiply(A=Add(A=@n41, B=@n38), B=@n42)
  n45 = Multiply(A=Add(A=ComponentMask(None=@n43) {r=true}, B=ComponentMask(None=@n43) {g=true}), B=Sine(None=Length(None=@n44)))
  n46 = ScalarParameter() {defaultValue=3, group=Texture, parameterName=Micro Noise Intensity, sortPriority=62}
  n47 = Max(
   A=Power(Base=NamedRerouteUsage(), Exp=ScalarParameter() {defaultValue=1, group=Texture, parameterName=Micro Noise Distribution, sortPriority=66})
   B=OneMinus(None=NamedRerouteUsage())
  )
  n48 = Multiply(A=LinearInterpolate(A=Power(Base=@n45, Exp=@n46), B=Constant() {r=1}, Alpha=Saturate(None=@n47)), B=@n37.Result)
  NamedRerouteDeclaration(None=QualitySwitch(
    Default=@n37.Result
    Epic=StaticSwitchParameter(True=@n48, False=@n37.Result) {desc=Expensive, group=Texture, parameterName=Use Micro Noise, sortPriority=50}
   )) {name=Texture, nodeColor=<rgba 1 0.75 0 1>}
  n49 = LocalPosition()
  n50 = ScalarParameter() {defaultValue=100, group=Texture, parameterName=Density Scale, sortPriority=-90}
  n51 = ParticleRandom()
  n52 = NamedRerouteDeclaration(None=Subtract(A=WorldPosition() {worldPositionShaderOffset=WPT_CameraRelative}.XYZ, B=ParticlePositionWS() {originType=CameraRelative})) {name=Camera Relative Local Position, nodeColor=<rgba 0 1 0.3906 1>}
  n53 = StaticSwitchParameter(
   True=ComponentMask(None=Transform(None=@n49.XYZ) {transformSourceType=TRANSFORMSOURCE_Instance}) {g=true, r=true}
   False=Add(A=Multiply(A=@n50, B=@n51) {constB=1000}, B=ComponentMask(None=@n52) {g=true, r=true})
  ) {
   desc=Expensive: Fog shapes will persist as they travel across multiple particles.
   group=Motion
   parameterName=Use Continuous Motion
   sortPriority=0
  }
  NamedRerouteDeclaration(None=@n53) {name=Texture Coordinates, nodeColor=<rgba 0.8984 1 0 1>}
  NamedRerouteDeclaration(None=@n10) {name=Slope, nodeColor=<rgba 0.4562 1 0.4 1>}
  NamedRerouteDeclaration(None=@n15) {name=RadiusGradient, nodeColor=<rgba 1 0.7969 0 1>}
  n54 = MaterialFunctionCall!(FadeOutTIme, ResetFadeInTime, FadeInTime) {materialFunction=/Game/Environments/TechArt/WindGustSystem/MF_WindGustsSystem}
  n55 = Power(Base=@n54.LinearTimeProgressionNormalized, Exp=ScalarParameter() {defaultValue=2, group=Motion, parameterName=Gust Acceleration})
  n56 = ScalarParameter() {defaultValue=1, group=Motion, parameterName=Gust Speed, sortPriority=31}
  n57 = ComponentMask(None=MaterialFunctionCall!(Position, IsTrees, IsGrass, IsMediumPlant, IsMediumPlant_2024Anim, IsMediumPlant2ndLevel, MediumPlant_BasePivotPostion) {materialFunction=/Game/Environments/Asteria/Foliage/Wind/MF_WindActorData}.WindI_Direction) {g=true, r=true}
  NamedRerouteDeclaration(None=Multiply(A=Multiply(A=@n55, B=@n56), B=Multiply(A=@n57) {constB=-1})) {name=Gust Vector, nodeColor=<rgba 1 1 1 1>}
  NamedRerouteDeclaration(None=Multiply(A=@n54.ResetFadeIn, B=@n54.ResetFadeIn)) {name=Gust Fade, nodeColor=<rgba 1 1 1 1>}
 ]
}
