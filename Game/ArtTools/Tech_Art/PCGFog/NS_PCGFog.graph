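# NS_PCGFog is a Niagara system that renders fog sprites from externally supplied instance data, named for use with PCG.
# Its single Minimal emitter passes the User.InstanceData float4 array to two local modules: SpawnFromInstanceData in emitter update and TransformFromInstanceData at particle spawn.
# Module names indicate the array drives particle spawning and transforms.
# A sprite renderer draws the particles, and User.FogMaterial exposes a material parameter, empty by default.
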NiagaraSystem /Game/ArtTools/Tech_Art/PCGFog/NS_PCGFog
dependencies {usedDataInterfaces [], usedDynamicInputs [], usedModules [&SystemState]}
emitters {Minimal {
  emitterUpdateScript [
   &EmitterState()
   &self:SpawnFromInstanceData(InstanceData:NiagaraDataInterfaceArrayFloat4=$User.InstanceData) {module=SpawnFromInstanceData}
  ]
  particleSpawnScript [&self:TransformFromInstanceData(InstanceData:NiagaraDataInterfaceArrayFloat4=$User.InstanceData) {module=TransformFromInstanceData}]
  rendererClasses [&NiagaraSpriteRendererProperties]
  renderers [{rendererClass=&NiagaraSpriteRendererProperties}]
 }}
summary {
 systemName=NS_PCGFog
 emitters [{emitterName=Minimal, rendererClasses [&NiagaraSpriteRendererProperties]}]
 userVariables [
  User.FogMaterial {
   type=MaterialInterface
   description=""
   defaultValue {struct=NiagaraExt_VariableValue_Object, value {object=null, objectClass=/Script/Engine.MaterialInterface}}
  }
  User.InstanceData {
   type=NiagaraDataInterfaceArrayFloat4
   description=""
   defaultValue {
    struct=NiagaraExt_VariableValue_DataInterface
    value {
     dataInterface=&self:NiagaraDataInterfaceArrayFloat4_0
     dataInterfaceClass=/Script/Niagara.NiagaraDataInterfaceArrayFloat4
    }
   }
  }
 ]
}
