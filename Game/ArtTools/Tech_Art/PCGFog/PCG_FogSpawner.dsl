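; PCG graph that scatters slope-aware fog points over landscape and cliffs, then passes them to a Blueprint that populates instance data.
; Surface samples avoid an excluded landscape layer, then stay flat at actor height or project onto surfaces.
; Projection uses ringed raycasts to separate cliff hits from terrain, averaging terrain normals for orientation.
; Points are optionally limited to landscape, filtered by slope, offset with slope compensation, and sized.
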(graph PCG_FogSpawner
  (params (Density :type Double :default 2.5e+02) (Size :type Double :default 2e+03) (VerticalOffset :type Double :default -2e+02) (TargetLandscape :type Bool :default true) (SlopeAvoidance :type Double :default 0.5) (ExcludedLandscapeLayer :type String :default "__LANDSCAPE_VISIBILITY__") (SlopeCompensation :type Double :default -4e+02) (SurfaceProjection :type Bool :default true))
  (bind GetActorData_0 (Get%20Actor%20Data :ActorSelector (struct :ActorFilter AllWorldActors :bMustOverlapSelf true :ActorSelection ByClass :ActorSelectionClass /Script/Landscape.LandscapeStreamingProxy :bSelectMultiple true) :bMergeSinglePointData true :Mode GetActorReference))
  (bind WorldRayHitQuery_1 (World%20Ray%20Hit%20Query :QueryParams (struct :ActorClassFilter Include :ActorClass /Hermes/Terrain/Cliffs/Blueprints/BP_Hermes_Cliff_PARENT.BP_Hermes_Cliff_PARENT_C :ActorFilterFromInput Include :bGetReferenceToActorHit true)))
  (bind SurfaceSampler_2 (Surface%20Sampler :Looseness 0.66 :PointExtents (1e+02 1e+02 1e+02) :PointSteepness 0.5 :PointExtents-param Density))
  (bind GetLandscapeData_0 (Get%20Landscape%20Data :SamplingProperties (struct)))
  (bind SurfaceSampler_1 (Surface%20Sampler :PointExtents (1e+02 1e+02 1e+02) :PointsPerSquaredMeter 0.5 :PointSteepness 0.5))
  (bind AttributeFilter_3 (Filter%20Attribute%20Elements :AttributeTypes (struct :DoubleValue 0.6) :bUseConstantThreshold true :TargetAttribute @Last :ThresholdAttribute @Last :TargetAttribute-param ExcludedLandscapeLayer))
  (bind ExtentsModifier_12 (Extents%20Modifier :Extents (3.0 3.0 3.0) :Mode Multiply))
  (bind Difference_8 (Difference :bDiffMetadata false :DensityFunction Binary))
  (bind ToPoint_9 (To%20Point))
  (bind Branch_0 (Branch :bOutputToB-param SurfaceProjection))
  (bind GetActorData_2 (Get%20Actor%20Data :Mode GetSinglePoint))
  (bind AttributeMathsOp_8 (Attribute%20Maths%20Op :InputSource1 $Position.Z :InputSource2 $Position.Z :Operation Set))
  (bind AttributeMathsOp_9 (Attribute%20Maths%20Op :InputSource1 $Rotation :InputSource2 $Position.Z :Operation Set))
  (bind ExtentsModifier_13 (Extents%20Modifier :Extents (5.0 5.0 5e+01)))
  (bind CreateAttribute_4 (Create%20Constant :AttributeTypes (struct :DoubleValue 1.5)))
  (bind CreatePoints_9 (Create%20Points :PointsToCreate [(struct :Transform (struct :Translation (1e+02 0.0 0.0))) (struct :Transform (struct :Translation (5e+01 86.6 0.0))) (struct :Transform (struct :Translation (-5e+01 86.6 0.0))) (struct :Transform (struct :Translation (-1e+02 0.0 0.0))) (struct :Transform (struct :Translation (-5e+01 -86.6 0.0))) (struct :Transform (struct :Translation (5e+01 -86.6 0.0)))]))
  (bind AttributeMathsOp_3 (Attribute%20Maths%20Op :InputSource1 $Position :Operation Multiply))
  (bind ExtentsModifier_5 (Extents%20Modifier :Extents (5.0 5.0 4e+01)))
  (bind CopyPoints_7 (Copy%20Points :AttributeInheritance TargetOnly :ColorInheritance Source :RotationInheritance Target :ScaleInheritance Source :SeedInheritance Target :TagInheritance Source))
  (bind CreateAttribute_20 (Create%20Constant :AttributeTypes (struct :Type Vector :VectorValue (0.0 0.0 1e+02))))
  (bind TransformPoints_19 (Transform%20Points))
  (bind CopyAttributes_5 (Copy%20Attributes :InputSource $Position :OutputTarget RayOrigin))
  (bind CopyAttributes_22 (Copy%20Attributes :InputSource $Rotation.Up :OutputTarget RayDirection))
  (bind CreateAttribute_24 (Create%20Constant :AttributeTypes (struct :DoubleValue -1.0)))
  (bind AttributeMathsOp_23 (Attribute%20Maths%20Op :Operation Multiply))
  (bind WorldRaycast_14 (World%20Raycast :bKeepOriginalPointOnMiss true :bOverrideRayDirections true :OriginInputAttribute RayOrigin :RaycastMode NormalizedWithLength :RayDirectionAttribute RayDirection :RayLength 2e+02 :WorldQueryParams (struct :SelectLandscapeHits Require)))
  (bind CopyAttributes_3 (Copy%20Attributes :InputSource $Rotation.Up :OutputTarget RayDirection))
  (bind AttributePartition_10 (Attribute%20Partition :PartitionAttributeSelectors [ID]))
  (bind AttributeReduce_0 (Attribute%20Reduce :bMergeOutputAttributes true :OutputAttributeName Up))
  (bind MakeRotatorAttribute_5 (Make%20Rotator%20Attribute :Operation MakeRotFromZ))
  (bind CopyAttributes_4 (Copy%20Attributes :InputSource Up :OutputTarget $Rotation))
  (bind CreateAttribute_7 (Create%20Constant :AttributeTypes (struct :DoubleValue 1.0)))
  (bind AttributeMathsOp_5 (Attribute%20Maths%20Op :InputSource1 $Density :Operation Set))
  (bind MergePoints_56 (Merge%20Points))
  (bind Branch_7 (Branch :bOutputToB-param TargetLandscape))
  (bind AttributePartition_18 (Attribute%20Partition :PartitionAttributeSelectors [ActorReference]))
  (bind GetAttributeFromPointIndex_19 (Get%20Attribute%20From%20Point%20Index :bForceOutputAttributeToBeInElementsDomain true :InputSource ActorReference :OutputAttributeName None))
  (bind ExecuteBlueprint_16 (Execute%20Blueprint :BlueprintElementInstance (class:/Game/Athena/Helios/Environments/Landscape/WorldBuildingTools/PCGGrass/PCGBP_StringContains.PCGBP_StringContains_C :Substring "Landscape") :BlueprintElementType /Game/Athena/Helios/Environments/Landscape/WorldBuildingTools/PCGGrass/PCGBP_StringContains.PCGBP_StringContains_C))
  (bind CopyAttributes_21 (Copy%20Attributes :InputSource @Last :OutputTarget IsLandscape))
  (bind MergePoints_22 (Merge%20Points))
  (bind AttributeFilter_23 (Filter%20Attribute%20Elements :AttributeTypes (struct :Type Boolean :BoolValue true) :bUseConstantThreshold true :Operator Equal :TargetAttribute @Last :ThresholdAttribute @Last))
  (bind NormalToDensity_24 (Normal%20To%20Density))
  (bind DensityFilter_25 (Density%20Filter :LowerBound 0.8 :LowerBound-param SlopeAvoidance))
  (bind MakeVectorAttribute_30 (Make%20Vector%20Attribute :OutputType Vector :Z VerticalOffset))
  (bind TransformPoints_27 (Transform%20Points))
  (bind AttributeMathsOp_16 (Attribute%20Maths%20Op :InputSource1 $Density :Operation OneMinus :OutputTarget SlopeCompensation))
  (bind AttributeMathsOp_17 (Attribute%20Maths%20Op :Operation Multiply :InB SlopeCompensation))
  (bind CopyAttributes_13 (Copy%20Attributes :InputSource $Rotation.Up :OutputTarget UpVector))
  (bind AttributeMathsOp_14 (Attribute%20Maths%20Op :Operation Multiply))
  (bind AttributeMathsOp_19 (Attribute%20Maths%20Op :InputSource1 $Position))
  (bind AttributeMathsOp_4 (Attribute%20Maths%20Op :InputSource1 $Scale :Operation Set :InB Size))
  (bind CreateAttribute_6 (Create%20Constant :AttributeTypes (struct :DoubleValue 2.0)))
  (bind AttributeMathsOp_15 (Attribute%20Maths%20Op :Operation Divide :InA Size))
  (bind ExtentsModifier_11 (Extents%20Modifier :Extents (1e+03 1e+03 1e+03)))
  (bind ExecuteBlueprint_1 (Execute%20Blueprint :BlueprintElementInstance (class:/Game/ArtTools/Tech_Art/PCGFog/BP_PCGFog_PopulateInstanceData.BP_PCGFog_PopulateInstanceData_C) :BlueprintElementType /Game/ArtTools/Tech_Art/PCGFog/BP_PCGFog_PopulateInstanceData.BP_PCGFog_PopulateInstanceData_C))
  (bind CreatePoints_0 (Create%20Points :PointsToCreate [(struct :Transform (struct :Translation (1e+02 0.0 0.0))) (struct :Transform (struct :Translation (-5e+01 86.6 0.0))) (struct :Transform (struct :Translation (-5e+01 -86.6 0.0)))]))
  (bind CreatePoints_2 (Create%20Points :PointsToCreate [(struct :Transform (struct :Translation (1e+02 0.0 0.0))) (struct :Transform (struct :Translation (30.9 95.1 0.0))) (struct :Transform (struct :Translation (-80.9 58.78 0.0))) (struct :Transform (struct :Translation (-80.9 -58.78 0.0))) (struct :Transform (struct :Translation (30.9 -95.1 0.0)))]))
  (bind CreatePoints_11 (Create%20Points :PointsToCreate [(struct :Transform (struct :Translation (1e+02 0.0 0.0))) (struct :Transform (struct :Translation (70.71 70.71 0.0))) (struct :Transform (struct :Translation (0.0 1e+02 0.0))) (struct :Transform (struct :Translation (-70.71 70.71 0.0))) (struct :Transform (struct :Translation (-1e+02 0.0 0.0))) (struct :Transform (struct :Translation (-70.71 -70.71 0.0))) (struct :Transform (struct :Translation (0.0 -1e+02 0.0))) (struct :Transform (struct :Translation (70.71 -70.71 0.0))) (struct :Steepness 1.0)]))
  (bind AttributeFilter_0 (Filter%20Attribute%20Elements :bUseConstantThreshold true :Operator Equal :ThresholdAttribute @Last))
  (bind ExtentsModifier_4 (Extents%20Modifier :Extents (2.0 2.0 1.0) :Mode Multiply))
  (bind CreateAttribute_25 (Create%20Constant :AttributeTypes (struct :DoubleValue 5.0)))
  (bind AttributeMathsOp_1 (Attribute%20Maths%20Op :InputSource1 $Position :Operation Multiply))
  (bind ExtentsModifier_3 (Extents%20Modifier :Extents (5.0 5.0 4e+01)))
  (bind CopyAttributes_7 (Copy%20Attributes :InputSource $Index :OutputTarget ID))
  (bind CopyPoints_5 (Copy%20Points :AttributeInheritance TargetOnly :ColorInheritance Source :RotationInheritance Target :ScaleInheritance Source :SeedInheritance Target :TagInheritance Source))
  (bind CreateAttribute_13 (Create%20Constant :AttributeTypes (struct :Type Vector :VectorValue (0.0 0.0 1e+02))))
  (bind TransformPoints_11 (Transform%20Points))
  (bind CopyAttributes_23 (Copy%20Attributes :InputSource $Position :OutputTarget RayOrigin))
  (bind CopyAttributes_15 (Copy%20Attributes :InputSource $Rotation.Up :OutputTarget RayDirection))
  (bind CreateAttribute_21 (Create%20Constant :AttributeTypes (struct :DoubleValue -1.0)))
  (bind AttributeMathsOp_18 (Attribute%20Maths%20Op :Operation Multiply))
  (bind WorldRaycast_9 (World%20Raycast :bKeepOriginalPointOnMiss true :bOverrideRayDirections true :OriginInputAttribute RayOrigin :RaycastMode NormalizedWithLength :RayDirectionAttribute RayDirection :RayLength 2e+02 :WorldQueryParams (struct :ActorClassFilter Require :ActorClass /Hermes/Terrain/Cliffs/Blueprints/BP_Hermes_Cliff_PARENT.BP_Hermes_Cliff_PARENT_C :SelectLandscapeHits Exclude :bGetReferenceToActorHit true)))
  (bind AttributePartition_26 (Attribute%20Partition :PartitionAttributeSelectors [ID]))
  (bind AttributeReduce_40 (Attribute%20Reduce :InputSource $Density :Operation Sum))
  (bind CreateAttribute_32 (Create%20Constant :AttributeTypes (struct :Type Integer32 :Int32Value 9 :DoubleValue 8.0)))
  (bind AttributeCompareOp_31 (Attribute%20Compare%20Op :OutputTarget AllHits))
  (bind AttributePartition_30 (Attribute%20Partition :PartitionAttributeSelectors [$Index]))
  (bind CopyAttributes_29 (Copy%20Attributes :InputSource @Last))
  (bind MergePoints_34 (Merge%20Points))
  (bind AttributeFilter_33 (Filter%20Attribute%20Elements :AttributeTypes (struct :Type Boolean :BoolValue true) :bUseConstantThreshold true :Operator Equal :TargetAttribute AllHits :ThresholdAttribute @Last))
  (bind AttributePartition_36 (Attribute%20Partition :PartitionAttributeSelectors [ID]))
  (bind AttributePartition_38 (Attribute%20Partition :PartitionAttributeSelectors [ID]))
  (bind GetAttributeFromPointIndex_41 (Get%20Attribute%20From%20Point%20Index :bForceOutputAttributeToBeInElementsDomain true :Index 8))
  (bind MergePoints_53 (Merge%20Points))
  (bind DeleteAttributes_43 (Delete%20Attributes :SelectedAttributes "AllHits"))
  (bind GetAttributeFromPointIndex_45 (Get%20Attribute%20From%20Point%20Index :bForceOutputAttributeToBeInElementsDomain true :Index 8))
  (bind MergePoints_46 (Merge%20Points))
  (bind DeleteAttributes_48 (Delete%20Attributes :SelectedAttributes "AllHits"))
  (bind WorldRaycast_50 (World%20Raycast :bKeepOriginalPointOnMiss true :bOverrideRayDirections true :OriginInputAttribute RayOrigin :RaycastMode NormalizedWithLength :RayDirectionAttribute RayDirection :RayLength 1e+04 :WorldQueryParams (struct :SelectLandscapeHits Require :bGetReferenceToActorHit true)))
  (reroute Cliffs DeleteAttributes_48)
  (reroute RejectedCliffCandidates WorldRaycast_50)
  (connect AttributeReduce_40 Out AttributeCompareOp_31 InA)
  (connect CreateAttribute_32 Out AttributeCompareOp_31 InB)
  (connect WorldRaycast_14 Out AttributeFilter_0 In)
  (connect SurfaceSampler_1 Out AttributeFilter_3 In)
  (connect MergePoints_22 Out AttributeFilter_23 In)
  (connect MergePoints_34 Out AttributeFilter_33 In)
  (connect CreatePoints_11 Out AttributeMathsOp_1 InA)
  (connect CreateAttribute_25 Out AttributeMathsOp_1 InB)
  (connect CreatePoints_9 Out AttributeMathsOp_3 InA)
  (connect CreateAttribute_4 Out AttributeMathsOp_3 InB)
  (connect AttributeMathsOp_19 Out AttributeMathsOp_4 InA)
  (connect CopyAttributes_4 Out AttributeMathsOp_5 InA)
  (connect CreateAttribute_7 Out AttributeMathsOp_5 InB)
  (connect Branch_0 Output%20A AttributeMathsOp_8 InA)
  (connect GetActorData_2 Out AttributeMathsOp_8 InB)
  (connect AttributeMathsOp_8 Out AttributeMathsOp_9 InA)
  (connect CopyAttributes_13 Out AttributeMathsOp_14 InA)
  (connect AttributeMathsOp_17 Out AttributeMathsOp_14 InB)
  (connect CreateAttribute_6 Out AttributeMathsOp_15 InB)
  (connect TransformPoints_27 Out AttributeMathsOp_16 In)
  (connect AttributeMathsOp_16 Out AttributeMathsOp_17 InA)
  (connect CopyAttributes_15 Out AttributeMathsOp_18 InA)
  (connect CreateAttribute_21 Out AttributeMathsOp_18 InB)
  (connect AttributeMathsOp_14 Out AttributeMathsOp_19 InA)
  (connect AttributeMathsOp_14 Out AttributeMathsOp_19 InB)
  (connect CopyAttributes_22 Out AttributeMathsOp_23 InA)
  (connect CreateAttribute_24 Out AttributeMathsOp_23 InB)
  (connect CopyAttributes_3 Out AttributePartition_10 In)
  (connect Branch_7 Output%20B AttributePartition_18 In)
  (connect WorldRaycast_9 Out AttributePartition_26 In)
  (connect AttributeCompareOp_31 Out AttributePartition_30 In)
  (connect AttributeFilter_33 InsideFilter AttributePartition_36 In)
  (connect AttributeFilter_33 OutsideFilter AttributePartition_38 In)
  (connect AttributePartition_10 Out AttributeReduce_0 In)
  (connect AttributePartition_26 Out AttributeReduce_40 In)
  (connect ToPoint_9 Out Branch_0 In)
  (connect AttributeMathsOp_9 Out Branch_7 In)
  (connect MergePoints_56 Out Branch_7 In)
  (connect WorldRaycast_14 Out CopyAttributes_3 Source)
  (connect WorldRaycast_14 Out CopyAttributes_3 Target)
  (connect MakeRotatorAttribute_5 Out CopyAttributes_4 Source)
  (connect ExtentsModifier_13 Out CopyAttributes_4 Target)
  (connect TransformPoints_19 Out CopyAttributes_5 Source)
  (connect CopyPoints_7 Out CopyAttributes_5 Target)
  (connect Branch_0 Output%20B CopyAttributes_7 Source)
  (connect Branch_0 Output%20B CopyAttributes_7 Target)
  (connect TransformPoints_27 Out CopyAttributes_13 Source)
  (connect TransformPoints_27 Out CopyAttributes_13 Target)
  (connect CopyAttributes_23 Out CopyAttributes_15 Source)
  (connect CopyAttributes_23 Out CopyAttributes_15 Target)
  (connect ExecuteBlueprint_16 Out CopyAttributes_21 Source)
  (connect AttributePartition_18 Out CopyAttributes_21 Target)
  (connect CopyAttributes_5 Out CopyAttributes_22 Source)
  (connect CopyAttributes_5 Out CopyAttributes_22 Target)
  (connect TransformPoints_11 Out CopyAttributes_23 Source)
  (connect CopyPoints_5 Out CopyAttributes_23 Target)
  (connect AttributePartition_30 Out CopyAttributes_29 Source)
  (connect AttributePartition_26 Out CopyAttributes_29 Target)
  (connect ExtentsModifier_3 Out CopyPoints_5 Source)
  (connect CopyAttributes_7 Out CopyPoints_5 Target)
  (connect ExtentsModifier_5 Out CopyPoints_7 Source)
  (connect RejectedCliffCandidates Out CopyPoints_7 Target)
  (connect MergePoints_53 Out DeleteAttributes_43 In)
  (connect MergePoints_46 Out DeleteAttributes_48 In)
  (connect NormalToDensity_24 Out DensityFilter_25 In)
  (connect ExtentsModifier_12 Out Difference_8 Differences)
  (connect SurfaceSampler_2 Out Difference_8 Source)
  (connect AttributeMathsOp_4 Out ExecuteBlueprint_1 InPoints)
  (connect GetAttributeFromPointIndex_19 Attribute ExecuteBlueprint_16 InString)
  (connect AttributeMathsOp_1 Out ExtentsModifier_3 In)
  (connect AttributeFilter_0 InsideFilter ExtentsModifier_4 In)
  (connect AttributeMathsOp_3 Out ExtentsModifier_5 In)
  (connect AttributeMathsOp_15 Out ExtentsModifier_11 Extents)
  (connect AttributeMathsOp_19 Out ExtentsModifier_11 In)
  (connect AttributeFilter_3 InsideFilter ExtentsModifier_12 In)
  (connect RejectedCliffCandidates Out ExtentsModifier_13 In)
  (connect AttributePartition_18 Out GetAttributeFromPointIndex_19 In)
  (connect AttributePartition_38 Out GetAttributeFromPointIndex_41 In)
  (connect AttributePartition_36 Out GetAttributeFromPointIndex_45 In)
  (connect AttributeReduce_0 Out MakeRotatorAttribute_5 Z)
  (connect CopyAttributes_21 Out MergePoints_22 In)
  (connect CopyAttributes_29 Out MergePoints_34 In)
  (connect GetAttributeFromPointIndex_45 Point MergePoints_46 In)
  (connect GetAttributeFromPointIndex_41 Point MergePoints_53 In)
  (connect AttributeMathsOp_5 Out MergePoints_56 In)
  (connect Cliffs Out MergePoints_56 In)
  (connect AttributeFilter_23 InsideFilter NormalToDensity_24 In)
  (connect Branch_7 Output%20A NormalToDensity_24 In)
  (connect GetLandscapeData_0 Out SurfaceSampler_1 Surface)
  (connect WorldRayHitQuery_1 Out SurfaceSampler_2 Surface)
  (connect Difference_8 Out ToPoint_9 In)
  (connect CopyPoints_5 Out TransformPoints_11 In)
  (connect CreateAttribute_13 Out TransformPoints_11 OffsetMax)
  (connect CreateAttribute_13 Out TransformPoints_11 OffsetMin)
  (connect CopyPoints_7 Out TransformPoints_19 In)
  (connect CreateAttribute_20 Out TransformPoints_19 OffsetMax)
  (connect CreateAttribute_20 Out TransformPoints_19 OffsetMin)
  (connect DensityFilter_25 Out TransformPoints_27 In)
  (connect MakeVectorAttribute_30 Out TransformPoints_27 OffsetMax)
  (connect MakeVectorAttribute_30 Out TransformPoints_27 OffsetMin)
  (connect AttributeMathsOp_18 Out WorldRaycast_9 Origins)
  (connect AttributeMathsOp_23 Out WorldRaycast_14 Origins)
  (connect DeleteAttributes_43 Out WorldRaycast_50 Origins)
  (connect GetActorData_0 Out WorldRayHitQuery_1 FilterActors))
